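# Animation Blueprint for the LexaBlitz axe pickaxe that adds procedural ambient motion to its sword and floating piece bones.
# Each update derives steady roll spins and two sine-wave offsets from real time.
# The AnimGraph applies these additively to the sword and piece_main bones, then simulates four piece bones with zero-gravity, linear-spring AnimDynamics.
# A "disable dynamics" curve sets curve_alpha, but no shown node reads it.
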AnimBlueprint /BRCosmetics/Weapons/FORT_Melee/Pickaxe_LexaBlitz/Meshes/LexaBlitz_Axe_AnimBP
target_skeleton=&/LexaBlitz_Axe_Skeleton
graphs [
 AnimGraph {nodes [
   AnimGraphNode_SubInput_0 = Input Pose()
   n1 = "Transform (Modify) Bone - Bone: sword_main"(
    ComponentPose=Local To Component(LocalPose=@AnimGraphNode_SubInput_0)
    Alpha=1
    Translation="0,0,0"
    Rotation=Get Ambient Rotation x 1() -> (Ambient Rotation x 1:Rotator)
    Scale="1,1,1"
   ) {
    rotationMode=BMM_Additive
    rotationSpace=BCS_BoneSpace
    translationMode=BMM_Additive
    translationSpace=BCS_BoneSpace
    alphaBoolBlend {blendOption=Linear}
    boneToModify {boneName=sword_main}
   }
   n2 = "Transform (Modify) Bone - Bone: sword"(ComponentPose=@n1, Alpha=1, Translation="0,0,0", Rotation=Get Ambient Rotation x 2() -> (Ambient Rotation x 2:Rotator), Scale="1,1,1") {
    rotationMode=BMM_Additive
    rotationSpace=BCS_BoneSpace
    translationMode=BMM_Additive
    translationSpace=BCS_BoneSpace
    alphaBoolBlend {blendOption=Linear}
    boneToModify {boneName=sword}
   }
   n3 = Get Pieces move() -> (Pieces move:Vector)
   n4 = "Transform (Modify) Bone - Bone: piece_main_1"!(ComponentPose=@n2, Alpha=0.8, Translation=@n3) -> (Pose) {translationMode=BMM_Additive, translationSpace=BCS_BoneSpace, alphaBoolBlend {blendOption=Linear}, boneToModify {boneName=piece_main_1}}
   n5 = "Transform (Modify) Bone - Bone: piece_main_3"!(ComponentPose=@n4, Alpha=0.8, Translation=@n3) -> (Pose) {translationMode=BMM_Additive, translationSpace=BCS_BoneSpace, alphaBoolBlend {blendOption=Linear}, boneToModify {boneName=piece_main_3}}
   n6 = Get Pieces move 2() -> (Pieces move 2:Vector)
   n7 = "Transform (Modify) Bone - Bone: piece_main_2"!(ComponentPose=@n5, Alpha=0.8, Translation=@n6) -> (Pose) {translationMode=BMM_Additive, translationSpace=BCS_BoneSpace, alphaBoolBlend {blendOption=Linear}, boneToModify {boneName=piece_main_2}}
   n8 = "Transform (Modify) Bone - Bone: piece_main_4"!(ComponentPose=@n7, Alpha=0.8, Translation=@n6) -> (Pose) {translationMode=BMM_Additive, translationSpace=BCS_BoneSpace, alphaBoolBlend {blendOption=Linear}, boneToModify {boneName=piece_main_4}}
   n9 = Anim Dynamics - Bone: piece_1!(SimulationSpace:AnimPhysSimSpaceType Enum, ComponentPose=@n8, Alpha=1, GravityOverride:Vector="0,0,0", ExternalForce="0,0,0") -> (Pose) {
    angularDampingOverride=1
    bGravityOverrideInSimSpace=true
    bLinearSpring=true
    bOverrideLinearDamping=true
    bUseGravityOverride=true
    linearDampingOverride=1
    linearSpringConstant=61.7856
    alphaBoolBlend {blendOption=Linear}
    boundBone {boneName=piece_1}
    componentLinearAccScale {z=0.07}
    componentLinearVelScale {z=0.07}
    physicsBodyDefinitions [{
      boundBone {boneName=piece_1}
      boxExtents=<20 20 20>
      constraintSetup {
       angularTargetAxis=AxisY
       coneAngle=1
       angularLimitsMax=<0 0 0>
       angularLimitsMin=<0 0 0>
       angularTarget=<1 0 0>
       linearAxesMax=<4 0 0>
       linearAxesMin=<-6 0 0>
      }
      localJointOffset=<0 0 0>
     }]
   }
   n10 = Anim Dynamics - Bone: piece_2!(SimulationSpace:AnimPhysSimSpaceType Enum, ComponentPose=@n9, Alpha=1, GravityOverride:Vector="0,0,0", ExternalForce="0,0,0") -> (Pose) {
    angularDampingOverride=1
    bGravityOverrideInSimSpace=true
    bLinearSpring=true
    bOverrideLinearDamping=true
    bUseGravityOverride=true
    linearDampingOverride=1
    linearSpringConstant=44.4181
    alphaBoolBlend {blendOption=Linear}
    boundBone {boneName=piece_2}
    componentLinearAccScale {z=0.07}
    componentLinearVelScale {z=0.07}
    physicsBodyDefinitions [{
      boundBone {boneName=piece_2}
      boxExtents=<20 20 20>
      constraintSetup {
       angularTargetAxis=AxisY
       coneAngle=1
       angularLimitsMax=<0 0 0>
       angularLimitsMin=<0 0 0>
       angularTarget=<1 0 0>
       linearAxesMax=<4 0 0>
       linearAxesMin=<-6 0 0>
      }
      localJointOffset=<0 0 0>
     }]
   }
   n11 = Anim Dynamics - Bone: piece_3!(SimulationSpace:AnimPhysSimSpaceType Enum, ComponentPose=@n10, Alpha=1, GravityOverride:Vector="0,0,0", ExternalForce="0,0,0") -> (Pose) {
    angularDampingOverride=1
    bGravityOverrideInSimSpace=true
    bLinearSpring=true
    bOverrideLinearDamping=true
    bUseGravityOverride=true
    linearDampingOverride=1
    linearSpringConstant=74.1914
    alphaBoolBlend {blendOption=Linear}
    boundBone {boneName=piece_3}
    componentLinearAccScale {y=0.03, z=0.07}
    componentLinearVelScale {y=0.02, z=0.07}
    physicsBodyDefinitions [{
      boundBone {boneName=piece_3}
      boxExtents=<20 20 20>
      constraintSetup {
       angularTargetAxis=AxisY
       coneAngle=1
       angularLimitsMax=<0 0 0>
       angularLimitsMin=<0 0 0>
       angularTarget=<1 0 0>
       linearAxesMax=<4 0 0>
       linearAxesMin=<-6 0 0>
      }
      localJointOffset=<0 0 0>
     }]
   }
   n12 = Anim Dynamics - Bone: piece_4!(SimulationSpace:AnimPhysSimSpaceType Enum, ComponentPose=@n11, Alpha=1, GravityOverride:Vector="0,0,0", ExternalForce="0,0,0") -> (Pose) {
    angularDampingOverride=1
    bGravityOverrideInSimSpace=true
    bLinearSpring=true
    bOverrideLinearDamping=true
    bUseGravityOverride=true
    linearDampingOverride=1
    linearSpringConstant=19.2738
    alphaBoolBlend {blendOption=Linear}
    boundBone {boneName=piece_4}
    componentLinearAccScale {z=0.07}
    componentLinearVelScale {z=0.07}
    physicsBodyDefinitions [{
      boundBone {boneName=piece_4}
      boxExtents=<20 20 20>
      constraintSetup {
       angularTargetAxis=AxisY
       coneAngle=1
       angularLimitsMax=<0 0 0>
       angularLimitsMin=<0 0 0>
       angularTarget=<1 0 0>
       linearAxesMax=<4 0 0>
       linearAxesMin=<-6 0 0>
      }
      localJointOffset=<0 0 0>
     }]
   }
   Output Pose(Result=Component To Local(ComponentPose=@n12)) {layerGroup=None}
  ]}
 EventGraph {nodes [
   Event BlueprintUpdateAnimation()
   > n1 = Sequence!(execute) -> (then_0, then_1, then_2:Exec)
   Set curve_alpha(
    execute=@n1.then_1
    curve_alpha:float=MapRangeClamped(Value=GetCurveValue(CurveName=disable dynamics), InRangeA=0, InRangeB=1, OutRangeA=1, OutRangeB=0)
   ) -> (Output_Get:float)
   n2 = GetRealTimeSeconds()
   n3 = "Sin (Radians)"(A=float + float(A=float * float(A=@n2, B=1.6) {is=title K2Node_PromotableOperator}))
   > Set Pieces move(Pieces move:Vector=MakeVector(X=float * float(A=@n3, B=1.5), Y=0)) -> (Output_Get:Vector)
   n4 = "Sin (Radians)"(A=float + float(A=float * float(A=@n2, B=2)))
   > Set Pieces move 2(Pieces move 2:Vector=MakeVector(X=float * float(A=@n4, B=1.2), Y=0)) -> (Output_Get:Vector)
   n5 = GetRealTimeSeconds()
   n6 = float / float(A=@n5, B=500) {is=title K2Node_CallFunction}
   Set Ambient Rotation x 1(
    execute=@n1.then_0
    Ambient Rotation x 1:Rotator=MakeRotator(Roll=float * float(A=@n6, B=10000) {is=title K2Node_CommutativeAssociativeBinaryOperator})
   ) -> (Output_Get:Rotator)
   n7 = float / float(A=@n5, B=1000)
   > Set Ambient Rotation x 2(Ambient Rotation x 2:Rotator=MakeRotator(Roll=float * float(A=@n7, B=10000))) -> (Output_Get:Rotator)
   TryGetPawnOwner()
  ]}
]
variables [
 Ambient Rotation x 1 {type=Rotator}
 Ambient Rotation x 2 {type=Rotator}
 curve_alpha {type=number}
 Pieces move {type=Vector}
 Pieces move 2 {type=Vector}
 Pieces move 3 {type=Vector}
]
